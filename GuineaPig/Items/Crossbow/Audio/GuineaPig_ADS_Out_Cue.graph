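# Sound cue for the GuineaPig crossbow that, by its name, plays when the player leaves aim-down-sights.
# A random node picks between two equally weighted waves, and a modulator varies the pitch slightly upward from normal on each play.
# It plays at low volume in the weapon foley sound class, using the weapon reload attenuation settings.
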SoundCue /GuineaPig/Items/Crossbow/Audio/GuineaPig_ADS_Out_Cue
root=Modulator_0
nodes [
 {id=Modulator_0, children [Random_0], properties {pitchMax=1.2, pitchMin=1}}
 {id=Random_0, children [WavePlayer_1,WavePlayer_2], properties {weights [1,1]}}
 {id=WavePlayer_1, properties {soundWaveAssetPtr=&/GuineaPig_ADS_Out_01}}
 {id=WavePlayer_2, properties {soundWaveAssetPtr=&/GuineaPig_ADS_Out_02}}
]
properties {
 attenuationSettings=&Fort_Weapon_Reload_Att
 maxDistance=2250
 soundClassObject=&Fort_WeaponFoley
 volumeMultiplier=0.1
}
