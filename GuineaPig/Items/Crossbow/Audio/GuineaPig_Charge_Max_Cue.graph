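# Sound cue for the GuineaPig crossbow reaching full charge, playing a one-shot with a sustained loop beneath it.
# A mixer layers the pitch-randomized full-charge wave with a quieter looping wave that an enveloper fades in briefly.
# The mix passes through charge-up attenuation and a third-person ranged weapon fire sound class.
# A concurrency override caps instances per owner, stopping the oldest first.
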SoundCue /GuineaPig/Items/Crossbow/Audio/GuineaPig_Charge_Max_Cue
root=SoundClass_2
nodes [
 {id=Attenuation_2, children [Mixer_0], properties {attenuationSettings=&/Att_GuineaPig_Chargeup}}
 {
  id=Enveloper_1
  children [Modulator_0]
  properties {pitchCurve {editorCurveData {keys [{time=0}]}}, volumeCurve {editorCurveData {keys [<at 0 0>,<at 0.2 1>]}}}
 }
 {id=Mixer_0, children [Modulator_2,Enveloper_1], properties {inputVolume [1,0.35]}}
 {id=Modulator_0, children [WavePlayer_6], properties {pitchMax=1.1, pitchMin=1.2, volumeMax=1, volumeMin=1}}
 {id=Modulator_2, children [WavePlayer_5], properties {pitchMax=1.2, pitchMin=1.1, volumeMax=1, volumeMin=1}}
 {
  id=SoundClass_2
  children [Attenuation_2]
  properties {soundClassOverride=&Fort_WeaponRanged_Fire_3P_PSM}
 }
 {id=WavePlayer_5, properties {soundWaveAssetPtr=&/GuineaPig_Charge_Full_01}}
 {id=WavePlayer_6, properties {bLooping=true, soundWaveAssetPtr=&/GuineaPig_Charge_Full_Loop_01}}
]
properties {
 bDebug=true
 bOverrideConcurrency=true
 maxDistance=8000
 soundClassObject=&Fort_WeaponRanged_Fire_3P_PSM
 volumeMultiplier=0.6
 concurrencyOverrides {bLimitToOwner=true, maxCount=3, resolutionRule=StopOldest}
}
