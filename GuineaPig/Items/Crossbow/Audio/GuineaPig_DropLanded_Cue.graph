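# Sound cue for the GuineaPig crossbow's drop-landed sound, evidently heard when the dropped item lands.
# A single landing wave player feeds a Modulator left at default settings, which is the cue's root.
# It uses the shared weapon-drop attenuation and loot-collect sound class.
# A concurrency override caps simultaneous plays at a few and blocks new ones beyond that instead of cutting off playing sounds.
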SoundCue /GuineaPig/Items/Crossbow/Audio/GuineaPig_DropLanded_Cue
root=Modulator_0
nodes [
 {id=Modulator_0, children [WavePlayer_1]}
 {id=WavePlayer_1, properties {soundWaveAssetPtr=&/GuineaPig_Drop_Land_01}}
]
properties {
 attenuationSettings=&Fort_Weapon_Drop_Att
 bOverrideConcurrency=true
 maxDistance=3000
 soundClassObject=&Fort_Loot_Collect
 volumeMultiplier=1
 attenuationOverrides {falloffDistance=1000, attenuationShapeExtents {x=0}}
 concurrencyOverrides {maxCount=3, resolutionRule=PreventNew}
}
