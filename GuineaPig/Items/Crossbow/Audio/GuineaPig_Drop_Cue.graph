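# Sound cue for the drop sound of the GuineaPig crossbow item.
# A single wave player feeds a Modulator at default settings, giving each play slight random pitch and volume variation.
# Shared weapon-drop attenuation spatializes it, the loot-collect sound class routes it, and the Weapon_Drop_StopOldest concurrency set limits overlapping drops.
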SoundCue /GuineaPig/Items/Crossbow/Audio/GuineaPig_Drop_Cue
root=Modulator_1
nodes [
 {id=Modulator_1, children [WavePlayer_1]}
 {id=WavePlayer_1, properties {soundWaveAssetPtr=&/GuineaPig_Drop_01}}
]
properties {
 attenuationSettings=&Fort_Weapon_Drop_Att
 maxDistance=3000
 soundClassObject=&Fort_Loot_Collect
 concurrencyOverrides {resolutionRule=PreventNew}
 concurrencySet [&Weapon_Drop_StopOldest]
}
