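# Sound cue for the dry-fire sound of the GuineaPig crossbow.
# A single wave player feeds a Modulator at default settings, which serves as the root.
# It plays in the weapon foley sound class at reduced volume, using the shared weapon reload attenuation asset.
# Overridden concurrency limits simultaneous plays per owner and sets a short retrigger time.
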SoundCue /GuineaPig/Items/Crossbow/Audio/GuineaPig_DryFire_Cue
root=Modulator_0
nodes [
 {id=Modulator_0, children [WavePlayer_1]}
 {id=WavePlayer_1, properties {soundWaveAssetPtr=&/GuineaPig_DryFire_01}}
]
properties {
 attenuationSettings=&Fort_Weapon_Reload_Att
 bOverrideConcurrency=true
 maxDistance=2250
 soundClassObject=&Fort_WeaponFoley
 volumeMultiplier=0.501
 attenuationOverrides {distanceAlgorithm=NaturalSound, falloffDistance=3000, attenuationShapeExtents {x=1000}}
 concurrencyOverrides {bLimitToOwner=true, maxCount=2, retriggerTime=0.2}
}
