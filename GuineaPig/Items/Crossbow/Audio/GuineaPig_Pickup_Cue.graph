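# Pickup sound cue for the GuineaPig crossbow item.
# A single wave player feeds the pickup sound through a Modulator left at its usual settings.
# It uses the shared weapon pickup attenuation and loot collect sound class.
# Concurrency references the PickupMaximum_Oldest set, alongside stored override values for a single instance that prevents new plays.
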SoundCue /GuineaPig/Items/Crossbow/Audio/GuineaPig_Pickup_Cue
root=Modulator_5
nodes [
 {id=Modulator_5, children [WavePlayer_0]}
 {id=WavePlayer_0, properties {soundWaveAssetPtr=&/GuineaPig_Pickup_01}}
]
properties {
 attenuationSettings=&Fort_Weapon_Pickup_Att
 maxDistance=3500
 soundClassObject=&Fort_Loot_Collect
 volumeMultiplier=1
 concurrencyOverrides {maxCount=1, resolutionRule=PreventNew}
 concurrencySet [&PickupMaximum_Oldest]
}
