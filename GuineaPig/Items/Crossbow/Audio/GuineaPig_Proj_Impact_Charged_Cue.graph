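# Sound cue for the impact of the GuineaPig crossbow's charged projectile, built as close, medium and distant layers.
# Each layer randomly picks one of four waves and passes through a modulator and its own explosion attenuation preset with overridden falloff.
# Sound class overrides put the close layer in a third-person ranged-fire class and the other two in the distant class.
# A mixer sums the layers, with the distant one at half volume.
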SoundCue /GuineaPig/Items/Crossbow/Audio/GuineaPig_Proj_Impact_Charged_Cue
root=Mixer_4
nodes [
 {
  id=Attenuation_0
  children [Modulator_2]
  properties {
   attenuationSettings=/Game/Sounds/Attenuation/Fort_Explosion_Distant_Athena_Att
   attenuationOverrides {falloffDistance=3000, attenuationShapeExtents {x=1000}}
  }
 }
 {
  id=Attenuation_4
  children [Random_5]
  properties {attenuationSettings=/Game/Sounds/Attenuation/Fort_Explosion_Close_Athena_Att, attenuationOverrides {falloffDistance=1600}}
 }
 {
  id=Attenuation_5
  children [Modulator_5]
  properties {
   attenuationSettings=/Game/Sounds/Attenuation/Fort_Explosion_Medium_Athena_Att
   attenuationOverrides {falloffDistance=3000, attenuationShapeExtents {x=1000}}
  }
 }
 {id=Mixer_4, children [SoundClass_3,SoundClass_2,SoundClass_0], properties {inputVolume [1,1,0.5]}}
 {id=Modulator_1, children [Attenuation_4]}
 {id=Modulator_2, children [Random_7], properties {pitchMax=1.1, pitchMin=0.9}}
 {id=Modulator_5, children [Random_6], properties {pitchMax=1.1, pitchMin=0.9}}
 {id=Random_5, children [WavePlayer_15, WavePlayer_16, WavePlayer_17, WavePlayer_18], properties {weights [1,1,1,1]}}
 {id=Random_6, children [WavePlayer_19, WavePlayer_20, WavePlayer_21, WavePlayer_22], properties {weights [1,1,1,1]}}
 {id=Random_7, children [WavePlayer_23, WavePlayer_24, WavePlayer_25, WavePlayer_26], properties {weights [1,1,1,1]}}
 {
  id=SoundClass_0
  children [Attenuation_0]
  properties {soundClassOverride=&Fort_WeaponRanged_Fire_Distant}
 }
 {
  id=SoundClass_2
  children [Attenuation_5]
  properties {soundClassOverride=&Fort_WeaponRanged_Fire_Distant}
 }
 {id=SoundClass_3, children [Modulator_1], properties {soundClassOverride=&Fort_WeaponRanged_Fire_3P_PSM}}
 {id=WavePlayer_15, properties {soundWaveAssetPtr=&/GuineaPig_Proj_Charged_Impact_Close_01}}
 {id=WavePlayer_16, properties {soundWaveAssetPtr=&/GuineaPig_Proj_Charged_Impact_Close_02}}
 {id=WavePlayer_17, properties {soundWaveAssetPtr=&/GuineaPig_Proj_Charged_Impact_Close_03}}
 {id=WavePlayer_18, properties {soundWaveAssetPtr=&/GuineaPig_Proj_Charged_Impact_Close_04}}
 {id=WavePlayer_19, properties {soundWaveAssetPtr=&/GuineaPig_Proj_Charged_Impact_Med_01}}
 {id=WavePlayer_20, properties {soundWaveAssetPtr=&/GuineaPig_Proj_Charged_Impact_Med_02}}
 {id=WavePlayer_21, properties {soundWaveAssetPtr=&/GuineaPig_Proj_Charged_Impact_Med_03}}
 {id=WavePlayer_22, properties {soundWaveAssetPtr=&/GuineaPig_Proj_Charged_Impact_Med_04}}
 {id=WavePlayer_23, properties {soundWaveAssetPtr=&/GuineaPig_Proj_Charged_Impact_Dist_01}}
 {id=WavePlayer_24, properties {soundWaveAssetPtr=&/GuineaPig_Proj_Charged_Impact_Dist_02}}
 {id=WavePlayer_25, properties {soundWaveAssetPtr=&/GuineaPig_Proj_Charged_Impact_Dist_03}}
 {id=WavePlayer_26, properties {soundWaveAssetPtr=&/GuineaPig_Proj_Charged_Impact_Dist_04}}
]
properties {maxDistance=27500}
