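# Sound cue for GuineaPig crossbow projectile impacts, named for water hits though it plays the Close impact waves.
# A Random node picks one of four equally weighted wave variations and feeds a Modulator at default settings for pitch and volume variation.
# The cue uses the projectile impact attenuation asset and the ranged weapon sound class.
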SoundCue /GuineaPig/Items/Crossbow/Audio/GuineaPig_Proj_Impact_Water_Cue
root=Modulator_0
nodes [
 {id=Modulator_0, children [Random_0]}
 {id=Random_0, children [WavePlayer_0,WavePlayer_1,WavePlayer_2,WavePlayer_3], properties {weights [1,1,1,1]}}
 {id=WavePlayer_0, properties {soundWaveAssetPtr=&/GuineaPig_Proj_Impact_Close_01}}
 {id=WavePlayer_1, properties {soundWaveAssetPtr=&/GuineaPig_Proj_Impact_Close_02}}
 {id=WavePlayer_2, properties {soundWaveAssetPtr=&/GuineaPig_Proj_Impact_Close_03}}
 {id=WavePlayer_3, properties {soundWaveAssetPtr=&/GuineaPig_Proj_Impact_Close_04}}
]
properties {
 attenuationSettings=&/Att_GuineaPig_Proj_Impact
 maxDistance=5000
 soundClassObject=&Fort_WeaponRanged
}
