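# Sound cue for the whoosh-by of a GuineaPig crossbow projectile, classed as a ranged weapon sound.
# A Random node picks one of five equally weighted whoosh waves and feeds a Modulator that varies pitch on each play.
# A dedicated attenuation asset handles spatial falloff.
# Overridden concurrency caps simultaneous plays at a low count, taking precedence over the assigned melee-weapon concurrency set.
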SoundCue /GuineaPig/Items/Crossbow/Audio/GuineaPig_Proj_WhooshBy_Cue
root=Modulator_1
nodes [
 {id=Modulator_1, children [Random_0], properties {pitchMax=1.2, pitchMin=0.8}}
 {id=Random_0, children [WavePlayer_1,WavePlayer_2,WavePlayer_3,WavePlayer_8,WavePlayer_9], properties {weights [1,1,1,1,1]}}
 {id=WavePlayer_1, properties {soundWaveAssetPtr=&/GuineaPig_Proj_Whoosh_By_01}}
 {id=WavePlayer_2, properties {soundWaveAssetPtr=&/GuineaPig_Proj_Whoosh_By_02}}
 {id=WavePlayer_3, properties {soundWaveAssetPtr=&/GuineaPig_Proj_Whoosh_By_03}}
 {id=WavePlayer_8, properties {soundWaveAssetPtr=&/GuineaPig_Proj_Whoosh_By_04}}
 {id=WavePlayer_9, properties {soundWaveAssetPtr=&/GuineaPig_Proj_Whoosh_By_05}}
]
properties {
 attenuationSettings=&/Att_GuineaPig_Proj_WhooshBy
 bOverrideConcurrency=true
 maxDistance=8800
 soundClassObject=&Fort_WeaponRanged
 volumeMultiplier=0.7
 attenuationOverrides {distanceAlgorithm=NaturalSound, falloffDistance=3000, attenuationShapeExtents {x=1000}}
 concurrencyOverrides {maxCount=2}
 concurrencySet [&Player_Melee_Weapon_Conc]
}
