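# One-shot stereo MetaSound for the start of the GuineaPig crossbow charge, layering charge-start and dry-fire samples.
# An affiliation check either triggers playback immediately or, when false, also waits for ChargePercent to cross a small threshold.
# Both wave players share a random pitch shift, and the dry-fire layer gets random gain before mixing.
# The sound finishes once both layers end, with owner-limited concurrency stopping the oldest instance.
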MetaSoundSource /GuineaPig/Items/Crossbow/Audio/MSS_GuineaPig_Charge_Start
inputs [
 Character.Affiliation {type=Enum:FortAudioAffiliation}
 Character.Health {type=Float}
 Character.UniqueID {type=Int32}
 Charge Start {type=WaveAsset}
 ChargePercent {type=Float}
 Dry Fire {type=WaveAsset}
 UE.Source.OnPlay {type=Trigger}
]
interfaces [Fort.Character:1.0, UE.OutputFormat.Stereo:1.0, UE.Source.OneShot:1.0, UE.Source:1.1]
outputs [
 "UE.OutputFormat.Stereo.Audio:0" {type=Audio}
 "UE.OutputFormat.Stereo.Audio:1" {type=Audio}
 UE.Source.OneShot.OnFinished {type=Trigger}
]
pages [Default {
  input_defaults {
   Character.Affiliation=0
   Character.Health=0
   Character.UniqueID=0
   Charge Start=&/GuineaPig_Charge_Start_01
   ChargePercent=0.11
   Dry Fire=&/GuineaPig_DryFire_01
   UE.Source.OnPlay=null
  }
  nodes [
   n1 = UE.Reroute(Value=UE.RandomFloat(Next=^UE.Source.OnPlay) {also {defaults {Max=1.05, Min=0.95}}}.Value)
   n2 = UE.RandomFloat(Next=^UE.Source.OnPlay) {also {defaults {Max=1.1, Min=1}}}
   n3 = &TriggerCompareAffiliation(A=^Character.Affiliation, Compare=^UE.Source.OnPlay)
   TriggerAccumulatorTriggerAccumulate2_1 = "TriggerAccumulator.Trigger Accumulate (2)"(In 0=@n3.False, In 1=UE.TriggerOnThreshold:Float(In=^ChargePercent) {also {defaults {Threshold=0.1}}}.Out)
   TriggerAnyTriggerAny2_0 = "TriggerAny.Trigger Any (2)"(In 1=@TriggerAccumulatorTriggerAccumulate2_1.Out, In 0=@n3.True)
   n4 = VariableAccessor.Trigger(Variable=VariableMutator.Trigger(Variable=InitVariable.Trigger().Variable, Value=@TriggerAnyTriggerAny2_0.Out).Variable)
   WavePlayer_1 = UE.Wave Player:Stereo(Wave Asset=^Dry Fire, Pitch Shift=@n2.Value, Play=VariableAccessor.Trigger(Variable=@n4.Variable).Value)
   WavePlayer_0 = UE.Wave Player:Stereo(Wave Asset=^Charge Start, Pitch Shift=@n2.Value, Play=@n4.Value)
   AudioMixerAudioMixerStereo2_0 = "AudioMixer.Audio Mixer (Stereo, 2)"(
    In 1 L=UE.Multiply:Audio by Float(AdditionalOperands=@n1.Value, PrimaryOperand=@WavePlayer_1.Out Left).Out
    In 1 R=UE.Multiply:Audio by Float(AdditionalOperands=@n1.Value, PrimaryOperand=@WavePlayer_1.Out Right).Out
    In 0 L=@WavePlayer_0.Out Left
    In 0 R=@WavePlayer_0.Out Right
   )
   TriggerAccumulatorTriggerAccumulate2_0 = "TriggerAccumulator.Trigger Accumulate (2)"(In 0=@WavePlayer_0.On Finished, In 1=@WavePlayer_1.On Finished)
  ]
  links [
   {dst="@UE.OutputFormat.Stereo.Audio:0", src=AudioMixerAudioMixerStereo2_0.Out L}
   {dst="@UE.OutputFormat.Stereo.Audio:1", src=AudioMixerAudioMixerStereo2_0.Out R}
   {dst="@UE.Source.OneShot.OnFinished", src=TriggerAccumulatorTriggerAccumulate2_0.Out}
  ]
 }]
properties {
 assetImportData=&self:AssetImportData
 attenuationSettings=&/Att_GuineaPig_Chargeup
 bOverrideConcurrency=true
 maxDistance=8000
 numChannels=2
 outputFormat=Stereo
 sampleRate=48000
 soundClassObject=&Fort_WeaponRanged_Fire_3P_PSM
 volume=0.3
 concurrencyOverrides {bLimitToOwner=true, maxCount=3, resolutionRule=StopOldest}
}
