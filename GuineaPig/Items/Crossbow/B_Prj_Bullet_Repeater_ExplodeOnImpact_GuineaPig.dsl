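; Projectile Blueprint for the GuineaPig repeater crossbow's explode-on-impact bolt, with a fiery Niagara trail and fake tracer mesh.
; On BeginPlay, cosmetic-capable instances feed muzzle location, start time and client state into tracer materials and trail.
; On stop, a fully charged shot triggers a sync or async explosion (scalable-float toggle) and a water or default gameplay cue by surface type.
; The trail receives the impact location, detaches, then deactivates after a short delay.
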(event EventBeginPlay
  (bind _projectilemesh (Variables|Default|GetProjectileMesh))
  (bind _ns_sniper_bullet_trail_fiery (Variables|Default|GetNS_Sniper_Bullet_Trail_Fiery))
  ;[1 1 1] Animate fake tracer mesh
  (Networking|CanExecuteCosmeticEvents
    (:True
      (bind _asfort_player_pawn (Utilities|Casting|CastToFortPlayerPawn (Variables|Actor|GetInstigator)))
      (Rendering|Material|SetVectorParameterValueonMaterials _projectilemesh "BulletOrigin" (Effects|GetMuzzleLocation (Variables|/Script/FortniteGame.FortPawn|GetCurrentWeapon _asfort_player_pawn) -1))
      (Rendering|Material|SetScalarParameterValueonMaterials _projectilemesh "StartTime" (Utilities|Time|GetGameTimeinSeconds))
      (Effects|Components|ParticleSystem|SetVectorParameter _ns_sniper_bullet_trail_fiery "MuzzlePosition" (Effects|GetMuzzleLocation (Variables|/Script/FortniteGame.FortPawn|GetCurrentWeapon _asfort_player_pawn) -1))
      (Effects|Components|ParticleSystem|SetBoolParameter _ns_sniper_bullet_trail_fiery "IsClient" (!= (Pawn|IsLocallyControlled _asfort_player_pawn) true)))))

(event EventOnStop
  (params (Hit :type Hit%20Result%20Structure%20%28by%20ref%29))
  (bind (_cachedhit_bblockinghit _cachedhit_binitialoverlap _cachedhit_time _cachedhit_distance _cachedhit_location) (Collision|BreakHitResult (Variables|GetCachedHit_bBlockingHit)))
  (bind _ns_sniper_bullet_trail_fiery (Variables|Default|GetNS_Sniper_Bullet_Trail_Fiery))
  (Variables|Default|SetCachedHit Hit)
  (Utilities|FlowControl|Sequence
    (:then_0
      (Effects|Components|ParticleSystem|SetVectorParameter _ns_sniper_bullet_trail_fiery "Impact Location" _cachedhit_location)
      (if (>= (Variables|FortProjectileBase|GetChargePercent) 1.0)
        (if (ScalableFloat|GetValueasBool (Variables|Default|GetbUseAsyncExplosion))
          (Delivery|DoExplosionAsync)
          (else
            (Delivery|DoExplosion)))
        (GameplayCues|ExecuteGameplayCue (select (Utilities|Enum|Equal(Enum) (Collision|GetSurfaceType (Variables|Default|GetCachedHit)) "SurfaceType8") (Variables|Default|GetWaterExplosionCueTag) (Variables|Default|GetExplosionCueTag)))))
    (:then_1
      (Transformation|DetachFromComponent _ns_sniper_bullet_trail_fiery "KeepWorld" "KeepWorld" "KeepWorld" true)
      (Effects|Components|ParticleSystem|SetBoolParameter _ns_sniper_bullet_trail_fiery "Kill Proj" true)
      (Utilities|FlowControl|Delay 0.2)
      (Components|Activation|Deactivate _ns_sniper_bullet_trail_fiery))))

(fn ConstructionScript
  (LOD|SetMaxDrawDistance (Variables|Default|GetProjectileMesh) (Variables|Default|GetDrawDistance)))
