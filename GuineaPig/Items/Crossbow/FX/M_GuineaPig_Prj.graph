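# Translucent unlit Niagara mesh-particle material for the GuineaPig crossbow projectile effect, apparently a stretching tracer.
# Particle-tinted dark-to-bright color follows a UV gradient and axis-aligned fresnel, with a U-based tail fade and HDR boost off mobile.
# Opacity reuses that mask with particle alpha, adding depth fade off mobile.
# An eased StartTime timer scales a red-vertex-color-masked offset toward a dynamic-parameter position, which reads X twice and never Y.
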Material /GuineaPig/Items/Crossbow/FX/M_GuineaPig_Prj
graph {
 properties {
  bUsedWithNiagaraMeshParticles=true
  blendMode=BLEND_Translucent
  pixelDepthOffsetMode=PDOM_Legacy
  shadingModel=MSM_Unlit
  twoSided=true
  usedShadingModels=Unlit
 }
 nodes [
  n1 = DynamicParameter() {paramNames [X,Y,Z,StartTime]}
  n2 = Subtract(A=AppendVector(A=AppendVector(A=@n1.X, B=@n1.X), B=@n1.Z), B=ActorPositionWS())
  Clamp(None=Divide(A=@n2) {constB=100}) {maxDefault=10}
  n3 = ComponentMask(None=TextureCoordinate()) {r=true}
  MaterialExpressionRemap_0 = MaterialXRemap!(
   None=Multiply(A=Power(Base=@n3) {constExponent=2.2})
   InputLow
   InputHigh
   TargetLow=ScalarParameter() {defaultValue=0.02, parameterName=TailFade_Min}
   TargetHigh
  ) {targetLowDefault=0.2}
  n4 = ParticleColor()
  n5 = TextureCoordinate()
  n6 = Transform(None=Constant3Vector() {constant=<rgba 1 0 0 0>}) {transformSourceType=TRANSFORMSOURCE_Local}
  n7 = Multiply(
   A=Power(Base=ComponentMask(None=@n5) {g=true}) {constExponent=1}
   B=MaterialFunctionCall(Axis=Normalize(VectorInput=@n6), Exponent=ScalarParameter() {defaultValue=2, parameterName=FresnelExponent}) {materialFunction=&AxisAlignedFresnel}.Fresnel
  )
  n8 = Multiply(
   A=@n4.RGB
   B=LinearInterpolate(
    A=VectorParameter() {parameterName=DarkColor, defaultValue=<rgba 1 0 0 1>}.RGB
    B=VectorParameter() {parameterName=BrightColor, defaultValue=<rgba 10 0 0 1>}.RGB
    Alpha=@n7
   )
  )
  n9 = MaterialFunctionCall(LDR Input=Clamp(None=@n8), HDR Tint=@n8, HDR Intensity Multiplier=ScalarParameter() {defaultValue=30, parameterName=HDRMult}) {materialFunction=&DeriveHDRfromLDR}
  MaterialFunctionCall(Emissive=Multiply(A=@MaterialExpressionRemap_0, B=ShadingPathSwitch(Default=@n9.RemappedRange, Mobile=@n8) {desc=Migrated from FeatureLevelSwitch})) {materialFunction=&MF_FX_EmissiveAdjustments}
  n10 = Divide(
   A=Subtract(A=Time() {period=5}, B=DynamicParameter() {paramNames [X,Y,Z,StartTime]}.StartTime)
   B=ScalarParameter() {defaultValue=0.01, group=Timestamp, parameterName=WPODuration}
  )
  n11 = Clamp(
   None=MaterialFunctionCall!(In=Clamp(None=@n10)) {materialFunction=/Game/Effects/Fort_Effects/Materials/_Functions/MF_Smoothstep_EaseOut}.Result
   Min=ScalarParameter() {defaultValue=0.1, parameterName=ScalingMin}
  ) {minDefault=0.1}
  Multiply(B=Clamp(None=Power(Base=@n11) {constExponent=0.3}))
  n12 = Multiply(A=LinearInterpolate(B=@n4.A, Alpha=ScalarParameter() {defaultValue=0.65, parameterName=Use_Particle_Alpha}) {constA=1}, B=@n7)
  n13 = ScalarParameter() {defaultValue=40, parameterName=DepthFadeDist}
  Multiply(
   A=ScalarParameter() {defaultValue=1, parameterName=Opacity}
   B=ShadingPathSwitch(Default=DepthFade(Opacity=@n12, FadeDistance=@n13), Mobile=@n12) {desc=Migrated from FeatureLevelSwitch}
  )
  OneMinus()
  OneMinus()
  Power(Base=Fresnel() {exponent=1}, Exp=ScalarParameter() {defaultValue=3})
  n14 = VertexColor()
  Reroute(None=Multiply(A=@n11, B=Multiply(A=@n14.R, B=@n2)))
  ScalarParameter() {parameterName=StartTime}
  VectorParameter() {parameterName=BulletOrigin}
  VertexNormalWS()
 ]
}
