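# Unlit additive, two-sided glow material for crossbow projectile effects on Niagara mesh particles.
# Color feeds an emissive-adjustment function, while opacity combines a ramp texture, optionally distorted by panning cloud noise, with lead, trail, depth and near-camera fades.
# Particle alpha can optionally modulate opacity.
# An offset chain adds camera offset, object-centered shrink and a time-eased, red-vertex-masked displacement from dynamic-parameter coordinates relative to actor position.
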Material /GuineaPig/Items/Crossbow/FX/M_GuineaPig_Proj_Glow
graph {
 properties {
  bUsedWithNiagaraMeshParticles=true
  blendMode=BLEND_Additive
  pixelDepthOffsetMode=PDOM_Legacy
  shadingModel=MSM_Unlit
  twoSided=true
  usedShadingModels=Unlit
 }
 nodes [
  ActorPositionWS()
  n1 = DynamicParameter() {paramNames [X,Y,Z,WPO_scale]}
  n2 = AppendVector(A=@n1.X, B=@n1.X)
  n3 = Normalize(VectorInput=Subtract(A=AppendVector(A=@n2, B=@n1.Z), B=ParticlePositionWS()))
  LinearInterpolate(B=Multiply(A=Multiply(A=@n1.WPO_scale, B=@n3), B=VertexColor().R), Alpha=ScalarParameter() {parameterName=Use_WPO})
  n4 = VectorParameter() {parameterName=Color, defaultValue=<rgba 1 0 0 0>}
  MaterialFunctionCall(Emissive=LinearInterpolate(A=Multiply(A=@n4.RGB) {constB=2}, B=@n4.RGB, Alpha=VertexColor().B)) {materialFunction=&MF_FX_EmissiveAdjustments}
  n5 = TextureCoordinate()
  n6 = ScalarParameter() {defaultValue=0.2, parameterName=Noise_Tile_X}
  n7 = ScalarParameter() {defaultValue=0.2, parameterName=Noise_Tile_Y}
  n8 = Panner(
   Coordinate=Multiply(A=TextureCoordinate(), B=AppendVector(A=@n6, B=@n7))
   Speed=AppendVector(A=ScalarParameter() {defaultValue=0.1, parameterName=Noise_Pan_X}, B=ScalarParameter() {defaultValue=-0.2, parameterName=Noise_Pan_Y})
  )
  n9 = Multiply(
   A=Power(
    Base=TextureSample(UVs=@n8) {texture=&AtmosphericCloudNoise01}.R
    Exp=ScalarParameter() {defaultValue=1, parameterName=Noise_Power}
   ) {constExponent=0.5}
   B=ScalarParameter() {defaultValue=0.2, parameterName=Noise_Intensity}
  ) {constB=0.4}
  n10 = Constant() {r=-0.1}
  n11 = Constant()
  n12 = LinearInterpolate(A=@n5, B=Add(A=Add(A=@n5, B=@n9), B=AppendVector(A=@n10, B=@n11)), Alpha=ScalarParameter() {parameterName=Use_Noise})
  n13 = Multiply(
   A=Power(
    Base=TextureSample(UVs=@n12) {texture=/Game/Global/Textures/Default/Masks_and_Gradients/T_Ramp_Virt}.R
    Exp=ScalarParameter() {defaultValue=1, parameterName=Noise_Line_Power}
   )
   B=ScalarParameter() {defaultValue=1, parameterName=Noise_Line_Mult}
  )
  n14 = TextureCoordinate()
  n15 = Multiply(A=Power(Base=ComponentMask(None=@n14) {g=true}) {constExponent=1})
  n16 = MaterialFunctionCall(Distance=ScalarParameter() {defaultValue=256, parameterName=Near Camera Fade}) {materialFunction=&Near_Camera_Fade}
  n17 = DepthFade(Opacity=@n16.Result, FadeDistance=ScalarParameter() {defaultValue=25, parameterName=DepthFadeDist})
  n18 = ComponentMask(None=TextureCoordinate()) {g=true}
  n19 = ScalarParameter() {defaultValue=2, parameterName=Lead_Fade_Power}
  n20 = Clamp(None=Multiply(A=Power(Base=@n18, Exp=@n19) {constExponent=4}, B=ScalarParameter() {defaultValue=1, parameterName=Lead_Fade_Mult}) {constB=2})
  n21 = OneMinus(None=ComponentMask(None=TextureCoordinate()) {g=true})
  n22 = ScalarParameter() {defaultValue=2.2, parameterName=Trail_Fade_Power}
  n23 = Clamp(None=Multiply(A=Power(Base=@n21, Exp=@n22), B=ScalarParameter() {defaultValue=1, parameterName=Trail_Fade_Mult}))
  n24 = Multiply(
   A=Clamp(None=Multiply(A=@n13, B=@n15))
   B=Multiply(A=ShadingPathSwitch(Default=@n17, Deferred=@n17, Mobile=@n16.Result) {desc=Migrated from FeatureLevelSwitch}, B=Multiply(A=@n20, B=@n23))
  )
  n25 = ParticleColor()
  Multiply(
   A=LinearInterpolate(A=@n24, B=Multiply(A=@n25.A, B=@n24), Alpha=ScalarParameter() {parameterName=Use_ParticleAlpha})
   B=ScalarParameter() {defaultValue=1, parameterName=Master_Opacity}
  )
  n26 = MaterialFunctionCall(Offset Amount=ScalarParameter() {defaultValue=5, parameterName=Camera Offset Amount}) {materialFunction=/Engine/Functions/Engine_MaterialFunctions02/WorldPositionOffset/CameraOffset}
  n27 = Multiply(A=Subtract(A=ObjectPositionWS(), B=WorldPosition().XYZ), B=OneMinus(None=ScalarParameter() {defaultValue=1, parameterName=Alpha}))
  n28 = Time() {period=5}
  n29 = DynamicParameter() {paramNames [X,Y,Z,WPO_scale]}
  n30 = Clamp(None=Divide(A=Subtract(A=@n28, B=@n29.WPO_scale), B=ScalarParameter() {defaultValue=0.2, group=Timestamp, parameterName=WPODuration}))
  n31 = DynamicParameter() {paramNames [X,Y,Z,WPO_scale]}
  n32 = AppendVector(A=@n31.X, B=@n31.X)
  n33 = Multiply(
   A=Clamp(
    None=MaterialFunctionCall!(In=@n30) {materialFunction=/Game/Effects/Fort_Effects/Materials/_Functions/MF_Smoothstep_EaseOut}.Result
    Min=ScalarParameter() {defaultValue=0.2, parameterName=ScalingMin}
   )
   B=Subtract(A=AppendVector(A=@n32, B=@n31.Z), B=ActorPositionWS())
  )
  n34 = VertexColor()
  Reroute(None=Add(A=Add(A=@n26.Camera Offset, B=@n27), B=Multiply(A=@n33, B=@n34.R)) {constB=0})
  ScalarParameter() {parameterName=StartTime}
  VectorParameter() {parameterName=BulletOrigin}
  VectorParameter() {parameterName=BulletOrigin}
 ]
}
