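# Additive, unlit, two-sided material for Niagara mesh particles, used as a crossbow trace effect in GuineaPig.
# Vertex-offset math interpolates along UV V between two dynamic-parameter points, adding sideways width and a mid-length vector bend.
# SplineThicken takes a depth-scaled width, while particle color feeds a center-desaturated HDR tint.
# Fade math combines UV masks, particle alpha, a dynamic Fade value, view angle and a world-space sphere mask.
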Material /GuineaPig/Items/Crossbow/FX/M_GuineaPig_Trace_WPO
graph {
 properties {
  bUsedWithNiagaraMeshParticles=true
  blendMode=BLEND_Additive
  pixelDepthOffsetMode=PDOM_Legacy
  shadingModel=MSM_Unlit
  twoSided=true
  usedShadingModels=Unlit
 }
 nodes [
  n1 = DynamicParameter() {parameterIndex=1, paramNames [X, Y, Z, N/A]}
  n2 = TransformPosition(None=TransformPosition(None=MaterialFunctionCall(X=@n1.X, Y=@n1.X, Z=@n1.X) {materialFunction=&MakeFloat3}.Result) {transformSourceType=TRANSFORMPOSSOURCE_World, transformType=TRANSFORMPOSSOURCE_TranslatedWorld}) {transformSourceType=TRANSFORMPOSSOURCE_TranslatedWorld, transformType=TRANSFORMPOSSOURCE_Instance}
  n3 = DynamicParameter() {paramNames [X,Y,Z,Fade]}
  n4 = MaterialFunctionCall(X=@n3.X, Y=@n3.X, Z=@n3.X) {materialFunction=&MakeFloat3}
  n5 = TextureCoordinate()
  n6 = LinearInterpolate(
   A=@n2
   B=TransformPosition(None=TransformPosition(None=@n4.Result) {transformSourceType=TRANSFORMPOSSOURCE_World, transformType=TRANSFORMPOSSOURCE_Instance}) {transformSourceType=TRANSFORMPOSSOURCE_TranslatedWorld, transformType=TRANSFORMPOSSOURCE_Instance}
   Alpha=ComponentMask(None=@n5) {g=true}
  ) {constB=0}
  n7 = Multiply(
   A=Subtract(A=ComponentMask(None=@n5) {r=true}) {constB=0.5}
   B=MaterialFunctionCall() {materialFunction=&TangentBasis}.X U
  )
  n8 = ComponentMask(None=TextureCoordinate()) {g=true}
  n9 = ScalarParameter() {defaultValue=1, parameterName=Param_14}
  n10 = LinearInterpolate(
   A=ScalarParameter() {parameterName=Param_11}
   B=ScalarParameter() {defaultValue=9}
   Alpha=SphereMask(
    A=Power(Base=@n8, Exp=@n9)
    B=Constant() {r=0.5}
    Radius=ScalarParameter() {defaultValue=0.5, parameterName=Param_9}
    Hardness=ScalarParameter() {parameterName=Param_10}
   )
  )
  n11 = VectorParameter()
  n12 = ComponentMask(None=TextureCoordinate()) {g=true}
  n13 = OneMinus(None=@n12)
  Add(A=Add(A=@n6, B=Multiply(A=@n7, B=@n10)), B=LinearInterpolate(B=Add(A=@n6, B=@n11.RGB), Alpha=Multiply(A=@n13, B=@n12)))
  n14 = Divide(A=PixelDepth().R, B=ScalarParameter() {defaultValue=768, parameterName=Param_8})
  MaterialFunctionCall(
   WidthBase=LinearInterpolate(B=ScalarParameter() {defaultValue=12, parameterName=Param_4}, Alpha=Saturate(None=@n14))
   WidthTip=ScalarParameter() {defaultValue=4, parameterName=Param_5}
  ) {materialFunction=&SplineThicken}
  n15 = ParticleColor()
  n16 = LinearInterpolate(
   A=@n15.RGB
   B=Desaturation(None=@n15.RGB, Fraction=Constant() {r=0.25})
   Alpha=SphereMask(
    A=TextureCoordinate()
    B=Constant() {r=0.5}
    Radius=ScalarParameter() {defaultValue=0.3, parameterName=Param_12}
    Hardness=ScalarParameter() {parameterName=Param_13}
   )
  )
  MaterialFunctionCall(LDR Input=Saturate(None=@n16), HDR Tint=@n16, HDR Intensity Multiplier=ScalarParameter() {defaultValue=0.5, parameterName=Param_15}) {materialFunction=&DeriveHDRfromLDR}
  n17 = TextureCoordinate()
  n18 = ComponentMask(None=@n17) {g=true}
  OneMinus(None=@n18)
  Reroute(None=@n15.RGB)
  n19 = ComponentMask(None=@n17) {r=true}
  n20 = Multiply(A=Multiply(A=OneMinus(None=@n19), B=@n19), B=Subtract(A=@n18, B=OneMinus(None=@n15.A)))
  n21 = OneMinus(None=ComponentMask(None=TextureCoordinate()) {g=true})
  n22 = DynamicParameter() {paramNames [X,Y,Z,Fade]}
  n23 = ScalarParameter() {defaultValue=0.5, parameterName=Param_3}
  n24 = Multiply(
   A=Multiply(A=Multiply(A=@n20, B=@n18), B=SphereMask(A=@n21, B=@n22.Fade, Radius=@n23) {hardnessPercent=0.1})
   B=MaterialFunctionCall("FadeSensitivity (optional)"=ScalarParameter() {defaultValue=1, parameterName=Param_1}) {materialFunction=&FadeBasedOnViewAngle}.Result
  )
  n25 = WorldPosition()
  n26 = ScalarParameter() {defaultValue=7000, parameterName=Param_7}
  n27 = ScalarParameter() {parameterName=Param_6}
  Reroute(None=Multiply(A=Multiply(A=@n15.A, B=@n24), B=SphereMask(A=@n25.XYZ, B=@n2, Radius=@n26, Hardness=@n27)))
  ScalarParameter() {parameterName=Param_2}
  VectorParameter() {parameterName=Param_1}
 ]
}
