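# Niagara projectile effect for the GuineaPig crossbow, drawing a bolt between the muzzle and owning component.
# Mesh emitters lerp from User.MuzzlePosition to owner position by age curves, orienting and stretching by distance to fake travel.
# User.IsClient keeps Projectile_Mesh and kills ClientSide_Proj, Fake_Proj_Mesh and a sprite glow; User.Kill Proj ends most emitters.
# A distance-culled light adds glow; bullet, particle and smoke-trail emitters are disabled.
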NiagaraSystem /GuineaPig/Items/Crossbow/FX/NS_GuineaPig_Prj
dependencies {
 usedDataInterfaces [&NiagaraDataInterfaceCurve, &NiagaraDataInterfaceVectorCurve]
 usedDynamicInputs []
 usedModules [
  &SystemState
  &self:SystemSpawnScript.SystemScriptSource.SystemScriptGraph.NiagaraNodeAssignment_0.SetVariables_#1
 ]
}
emitters {
 Bullet {
  bEnabled=false
  emitterUpdateScript [
   &EmitterState(Inactive Response=::NewEnumerator1, Life Cycle Mode=::NewEnumerator1, Loop Behavior=::NewEnumerator1, Loop Duration=1, UseLoopDelay=true)
   &SpawnBurst_Instantaneous(Spawn Time=0.0025, Use Spawn Probability=false, UseLoopCountLimit=false)
  ]
  particleSpawnScript [
   /SetVariables(Particles.SpawnRotationVector=<2 0 0>) {isSetParameters=true}
   &InitializeParticle(Mesh Scale Mode=::NewEnumerator3, Mesh Uniform Scale=1.7, UsePositionOffset=false, Write Material Random=true)
   &Color(Color=&MakeLinearColorFromVectorAndFloat("Vector (RGB)"=<1 1 1>, "Float (Alpha)"=&ClampFloat(Float=1)))
  ]
  particleUpdateScript [
   &ParticleState()
   &UpdateMeshOrientation(Rotation Vector=$Particles.SpawnRotationVector)
   &DynamicMaterialParameters(Module.Param0WriteEnabled=true, Module.Param1WriteEnabled=true, Module.Param2WriteEnabled=true, Module.Param3WriteEnabled=true)
  ]
  rendererClasses [&NiagaraMeshRendererProperties]
  renderers [{rendererClass=&NiagaraMeshRendererProperties, rendererIndex=0}]
 }
 ClientSide_Proj {
  emitterUpdateScript [
   &EmitterState(Life Cycle Mode=::NewEnumerator1, Loop Behavior=::NewEnumerator1, Loop Duration Mode=::NewEnumerator1, UseLoopDelay=false)
   &SpawnBurst_Instantaneous(Use Spawn Probability=false, UseLoopCountLimit=false)
  ]
  particleSpawnScript [
   &InitializeParticle(
    Color=<rgba 0.75 0.75 0.75 0.5>
    Mesh Scale=<3.25 0.6 0.6>
    Mesh Scale Mode=::NewEnumerator4
    Sprite Size Mode=::NewEnumerator3
    Uniform Sprite Size=5000
    UsePositionOffset=false
    Write Material Random=true
   )
   &SystemLocation(MaskSpawnGroup=false, Offset=<50 0 0>, Offset Coordinate Space=::World) {enabled=false}
   &InitializeParticle(
    Color Mode=::NewEnumerator0
    Lifetime=&Lerp_Float(
     A=0.1
     B=1
     Alpha=&RemapRange(Input Value To Remap=$User.ScanLength, Input Min=0, Input Max=1000, Output Min=0, Output Max=1)
    )
    Mesh Scale Mode=::NewEnumerator4
    Sprite Size Mode=::NewEnumerator3
    UsePositionOffset=false
    Write Material Random=true
   ) {enabled=false, module=InitializeParticle001}
  ]
  particleUpdateScript [
   &ParticleState(Kill Particles When Lifetime Has Elapsed=false, Let Infinitely Lived Particles Die When Emitter Deactivates=true)
   &DynamicMaterialParameters(
    Index 0 Param 1=$System.Muzzle_Pos_X
    Index 0 Param 2=$System.Muzzle_Pos_Y
    Index 0 Param 3=$System.Muzzle_Pos_Z
    Index 0 Param 4=&FloatFromCurve(FloatCurve=DataInterface() {propertyValues {Curve {keys [<at 0 0>,<at 0.1 1>,<at 1 10>]}}}, CurveIndex=$Particles.Age, Scale Curve=1000)
    Module.Param0WriteEnabled=true
    Module.Param1WriteEnabled=true
    Module.Param2WriteEnabled=true
    Module.Param3WriteEnabled=true
   )
   &ScaleMeshSize(Scale Factor=&MakeVector(
     X=&ClampFloat(
      Float=&Divide_Float(
       A=&VectorLength(Vector=&Subtract_Vector(
         A=Linked() {linkedVariable {name=User.MuzzlePosition, type=NiagaraPosition}}
         B=Linked() {linkedVariable {name=Engine.Owner.Position, type=NiagaraPosition}}
        ))
       B=160
      )
      Min=0.1
      Max=10
     )
     Y=1
     Z=1
    ))
   &UpdateMeshOrientation(
    Facing Direction=&NormalizeVector(
     Vector To Normalize=&Subtract_Vector(
      A=Linked() {linkedVariable {name=Engine.Owner.Position, type=NiagaraPosition}}
      B=Linked() {linkedVariable {name=User.MuzzlePosition, type=NiagaraPosition}}
     )
     Safe Normalize=true
     Safe Fallback Threshold=0
     Safe Fallback Vector=<0 0 0.001>
    )
    Facing Interpolation=false
    Orientation Method=::NewEnumerator1
    Override Quat=false
    TransformFacing=false
    TransformUp=false
   )
   /SetVariables(Particles.Position=&Lerp_Vector(
     A=$User.MuzzlePosition001
     B=Linked() {linkedVariable {name=Engine.Owner.Position, type=NiagaraPosition}}
     Alpha=&Lerp_Float(
      A=&FloatFromCurve(
       FloatCurve=DataInterface() {propertyValues {Curve {keys [<at 0 0>, {arriveTangent=15.1596, leaveTangent=15.1596, time=0.1, value=1}]}}}
       CurveIndex=$Particles.Age
      )
      B=&FloatFromCurve(
       FloatCurve=DataInterface() {propertyValues {Curve {keys [{arriveTangent=0.7022, time=0, value=0}, {arriveTangent=5.581, leaveTangent=5.581, time=1, value=1}]}}}
       CurveIndex=$Particles.Age
      )
      Alpha=&RemapRange(
       Input Value To Remap=&VectorLength(Vector=&Subtract_Vector(A=$User.MuzzlePosition001, B=Linked() {linkedVariable {name=Engine.Owner.Position, type=NiagaraPosition}}))
       Input Min=100
       Input Max=15000
       Output Min=0
       Output Max=1
      )
     )
    )) {enabled=false, isSetParameters=true, module=SetVariables_#2}
   /SetVariables(Particles.Position=&Lerp_Vector(
     A=Linked() {linkedVariable {name=User.MuzzlePosition, type=NiagaraPosition}}
     B=Linked() {linkedVariable {name=Engine.Owner.Position, type=NiagaraPosition}}
     Alpha=&FloatFromCurve(
      FloatCurve=DataInterface() {propertyValues {Curve {keys [{arriveTangent=2.0655, time=0, value=0}, {arriveTangent=0.2891, leaveTangent=0.2891, time=0.17, value=1}]}}}
      CurveIndex=$Particles.Age
     )
    )) {isSetParameters=true}
   &ScaleMeshSize(Scale Factor=&VectorFromFloat(Value=&Lerp_Float(
      A=0
      B=1
      Alpha=&FloatFromCurve(
       FloatCurve=DataInterface() {propertyValues {Curve {keys [
           {interpMode=RCIM_Cubic, leaveTangent=49.8975, tangentMode=RCTM_Break, time=0.015, value=0.125}
           {tangentMode=RCTM_User, time=0.07, value=1.15}
          ]}}}
       CurveIndex=$Particles.Age
      )
     ))) {module=ScaleMeshSize001}
   &ScaleColor(
    Scale Alpha=&FloatFromCurve(FloatCurve=DataInterface() {propertyValues {Curve {keys [<at 0.15 0>,<at 0.25 0>]}}}, CurveIndex=$Particles.Age)
    ScaleA=true
    ScaleRGB=true
   )
   &KillParticles(Kill Particles=&Boolean_OrOperation(
     Bool1=&Boolean_AndOperation(Bool1=$User.Kill Proj, Bool2=&SetBoolByFloatComparison(A=$Particles.Age, B=0.15))
     Bool2=$User.IsClient
    ))
  ]
  rendererClasses [&NiagaraSpriteRendererProperties, &NiagaraMeshRendererProperties]
  renderers [
   {rendererClass=&NiagaraSpriteRendererProperties}
   {rendererClass=&NiagaraMeshRendererProperties, rendererIndex=1}
  ]
 }
 Fake_Proj_Mesh {
  emitterUpdateScript [
   &EmitterState()
   &SpawnRate(SpawnRate=20, Use Spawn Probability=false) {enabled=false}
   &SpawnBurst_Instantaneous(Use Spawn Probability=false, UseLoopCountLimit=false)
  ]
  particleSpawnScript [
   &InitializeParticle(
    Color=<rgba 50 0 0 5>
    Lifetime=0.125
    Mesh Scale Mode=::NewEnumerator3
    Position=/Niagara/DynamicInputs/Helpers/SimulationPosition()
    Position Mode=::NewEnumerator1
    UsePositionOffset=false
    Write Material Random=true
   )
   /SetVariables(Particles.End Position:NiagaraPosition=&SubtractVectorFromPosition(Position=$User.MuzzlePosition, Vector=<10 0 0>)) {isSetParameters=true}
  ]
  particleUpdateScript [
   /SetVariables(Particles.Start Position:NiagaraPosition=$Engine.Owner.Position) {isSetParameters=true}
   /SetVariables(Particles.Quat:Quat4f=&OrientMeshToVector(
     Look At Direction=&SubtractVectorFromPosition(Position=$Engine.Owner.Position, Vector=Linked() {linkedVariable {name=Particles.End Position, type=NiagaraPosition}})
     Reference Axis Vector=<1 0 0>
    )) {enabled=false, isSetParameters=true}
   &ParticleState()
   &UpdateMeshOrientation(
    Facing Direction=&SubtractVectorFromPosition(Position=$Engine.Owner.Position, Vector=Linked() {linkedVariable {name=User.MuzzlePosition, type=NiagaraPosition}})
    Facing Interpolation=false
    Orientation Method=::NewEnumerator1
    Override Quat=false
    TransformFacing=false
    TransformUp=false
   ) {enabled=false}
   &ScaleMeshSize(Scale Factor=&VectorFromFloat(Value=&Lerp_Float(
      A=0
      B=1
      Alpha=&FloatFromCurve(FloatCurve=DataInterface() {propertyValues {Curve {keys [<at 0.35 0>,<at 0.4 1>]}}})
     )))
   &DynamicMaterialParameters(
    Index 0 Param 1=&MakeFloatFromVector(Vector=Linked() {linkedVariable {name=Engine.Owner.Position, type=NiagaraPosition}}, Channel=::NewEnumerator0)
    Index 0 Param 2=&MakeFloatFromVector(Vector=Linked() {linkedVariable {name=Engine.Owner.Position, type=NiagaraPosition}}, Channel=::NewEnumerator1)
    Index 0 Param 3=&MakeFloatFromVector(Vector=Linked() {linkedVariable {name=Engine.Owner.Position, type=NiagaraPosition}}, Channel=::NewEnumerator2)
    Index 0 Param 4=&FloatFromCurve(FloatCurve=DataInterface() {propertyValues {Curve {keys [<at 0 1>,<at 1 0>]}}})
    Index 1 Param 1=&MakeFloatFromVector(Vector=$Particles.Initial.End Position, Channel=::NewEnumerator0)
    Index 1 Param 2=&MakeFloatFromVector(Vector=$Particles.Initial.End Position, Channel=::NewEnumerator1)
    Index 1 Param 3=&MakeFloatFromVector(Vector=$Particles.Initial.End Position, Channel=::NewEnumerator2)
    Module.Index0Param0Write=true
    Module.Index0Param1Write=true
    Module.Index0Param2Write=true
    Module.Index0Param3Write=true
    Module.Index1Param0Write=true
    Module.Index1Param1Write=true
    Module.Index1Param2Write=true
    Module.Index1Param3Write=false
    Write Parameter Index 1=true
   )
   &ScaleColor(
    Scale Alpha=&FloatFromCurve(FloatCurve=DataInterface() {propertyValues {Curve {keys [<at 0 1>,<at 1 0>]}}})
    ScaleA=true
    ScaleRGB=true
   )
   &KillParticles(Kill Particles=&Boolean_OrOperation(
     Bool1=&SetBoolByFloatComparison(A=&DistanceBetweenPositions(A=$Particles.Start Position, B=$Particles.End Position), B=7500)
     Bool2=$User.IsClient
    ))
  ]
  rendererClasses [&NiagaraMeshRendererProperties]
  renderers [
   {rendererClass=&NiagaraMeshRendererProperties, rendererIndex=0}
   {rendererClass=&NiagaraMeshRendererProperties, rendererIndex=1}
   {rendererClass=&NiagaraMeshRendererProperties, rendererIndex=2}
  ]
 }
 Light {
  emitterUpdateScript [
   &EmitterState(
    Enable Distance Culling=true
    Life Cycle Mode=::NewEnumerator1
    Loop Behavior=::NewEnumerator1
    Loop Duration Mode=::NewEnumerator1
    MaxDistance=2500
    Module.bUseMaxDistance=true
    Module.bUseMinDistance=false
    Scalability Mode=::NewEnumerator1
    UseLoopDelay=false
   )
   &SpawnBurst_Instantaneous(Spawn Count=1, Use Spawn Probability=false)
  ]
  particleSpawnScript [
   &InitializeParticle(Lifetime=0.5, UsePositionOffset=false, Write Material Random=true) {module=InitializeParticle001}
   /SetVariables(Particles.LightPositionNew:Vector3f=&Add_Vector(
     A=Linked() {linkedVariable {name=Particles.Position, type=NiagaraPosition}}
     B=/Niagara/DynamicInputs/Transforms/TransformPosition(Position=<25 0 0>, Source Space=::Local, Destination Space=::Simulation)
    )) {enabled=false, isSetParameters=true, module=SetVariables_#3}
   /SetVariables() {isSetParameters=true}
  ]
  particleUpdateScript [
   /SetVariables(Particles.Light_Color:LinearColor=&MultiplyLinearColor(
     Linear Color=<rgba 1 0.0394 0 1>
     Scale Factor=&MakeVector4_FromFloat(Float=&Multiply_Float(
       A=20
       B=&FloatFromCurve(FloatCurve=DataInterface() {propertyValues {Curve {keys [
            {interpMode=RCIM_Cubic, tangentMode=RCTM_User, time=0, value=0}
            {interpMode=RCIM_Cubic, tangentMode=RCTM_User, time=0.2, value=1}
            {tangentMode=RCTM_User, time=1, value=0}
           ]}}})
      ))
    )) {isSetParameters=true}
   /SetVariables(Particles.LightRadius=&Multiply_Float(
     A=500
     B=&ClampFloat(
      Float=&FN_LOD_Distance_Remap(Start Distance=0, End Distance=8000, Minimum Output=1, Maximum Output=2)
      Min=1
      Max=2
     )
    )) {isSetParameters=true}
   &ParticleState(Kill Particles When Lifetime Has Elapsed=false)
   /SetVariables(Particles.Position=&Lerp_Vector(
     A=Linked() {linkedVariable {name=User.MuzzlePosition, type=NiagaraPosition}}
     B=Linked() {linkedVariable {name=Engine.Owner.Position, type=NiagaraPosition}}
     Alpha=&FloatFromCurve(
      FloatCurve=DataInterface() {propertyValues {Curve {keys [
          {arriveTangent=2.265, interpMode=RCIM_Cubic, leaveTangent=3.3003, tangentMode=RCTM_Break, time=0, value=0}
          {arriveTangent=0.2891, leaveTangent=0.2891, time=1, value=1}
         ]}}}
      CurveIndex=$Particles.Age
     )
    )) {enabled=false, isSetParameters=true, module=SetVariables_#4}
   &KillParticles(Kill Particles=$User.Kill Proj)
  ]
  rendererClasses [&NiagaraLightRendererProperties]
  renderers [{rendererClass=&NiagaraLightRendererProperties, rendererIndex=0}]
 }
 Light001 {
  emitterUpdateScript [
   &EmitterState(
    Enable Distance Culling=true
    Life Cycle Mode=::NewEnumerator1
    Loop Behavior=::NewEnumerator1
    Loop Duration Mode=::NewEnumerator1
    MaxDistance=2500
    Module.bUseMaxDistance=true
    Module.bUseMinDistance=false
    Scalability Mode=::NewEnumerator1
    UseLoopDelay=false
   )
   &SpawnBurst_Instantaneous(Spawn Count=1, Use Spawn Probability=false)
  ]
  particleSpawnScript [
   &InitializeParticle(
    Color=<rgba 10 0 0.2787 1>
    Lifetime=0.5
    Sprite Size Mode=::NewEnumerator3
    Uniform Sprite Size=20
    UsePositionOffset=false
    Write Material Random=true
   ) {module=InitializeParticle001}
   /SetVariables(Particles.LightPositionNew:Vector3f=&Add_Vector(
     A=Linked() {linkedVariable {name=Particles.Position, type=NiagaraPosition}}
     B=&TransformPosition(Position=<25 0 0>, Source Space=::Local, Destination Space=::Simulation)
    )) {enabled=false, isSetParameters=true, module=SetVariables_#3}
  ]
  particleUpdateScript [
   &ParticleState(Kill Particles When Lifetime Has Elapsed=false)
   &ScaleSpriteSize()
   /SetVariables(Particles.Position=&Lerp_Vector(
     A=Linked() {linkedVariable {name=User.MuzzlePosition, type=NiagaraPosition}}
     B=Linked() {linkedVariable {name=Engine.Owner.Position, type=NiagaraPosition}}
     Alpha=&FloatFromCurve(
      FloatCurve=DataInterface() {propertyValues {Curve {keys [
          {arriveTangent=2.265, interpMode=RCIM_Cubic, leaveTangent=3.3003, tangentMode=RCTM_Break, time=0, value=0}
          {arriveTangent=0.2891, leaveTangent=0.2891, time=1, value=1}
         ]}}}
      CurveIndex=$Particles.Age
     )
    )) {enabled=false, isSetParameters=true, module=SetVariables_#4}
   &KillParticles(Kill Particles=&Boolean_OrOperation(Bool1=$User.Kill Proj, Bool2=$User.IsClient))
   &ScaleColor(
    Scale Alpha=&FloatFromCurve(FloatCurve=DataInterface() {propertyValues {Curve {keys [{tangentMode=RCTM_User, time=0, value=0}, {tangentMode=RCTM_User, time=1, value=1}]}}})
    ScaleA=true
    ScaleRGB=true
   )
  ]
  rendererClasses [&NiagaraSpriteRendererProperties]
  renderers [{rendererClass=&NiagaraSpriteRendererProperties}]
 }
 Particle_Emitter {
  bEnabled=false
  emitterUpdateScript [
   &EmitterState(Inactive Response=::NewEnumerator1, Life Cycle Mode=::NewEnumerator1, Loop Behavior=::NewEnumerator1, Loop Duration=1, UseLoopDelay=true)
   &SpawnBurst_Instantaneous(Spawn Time=0.0025, Use Spawn Probability=false, UseLoopCountLimit=false)
  ]
  particleSpawnScript [
   &InitializeParticle(
    Color=<rgba 1 0.01 0 1>
    Lifetime=10
    Mesh Scale=&Multiply_VectorByFloat(Vector=<1.2 0.6 0.6>, Float=1)
    Mesh Scale Mode=::NewEnumerator4
    Position Mode=::NewEnumerator1
    UsePositionOffset=false
    Write Material Random=true
   )
   /SetVariables(Particles.MeshOrientation=&MakeQuatFromAxisAngle(Axis=<0 1 0>, AngleInDegrees=180)) {isSetParameters=true}
   &Color(Color=&MakeLinearColorFromVectorAndFloat("Vector (RGB)"=<12 2.1688 0.4912>, "Float (Alpha)"=&ClampFloat(Float=0.5))) {enabled=false}
  ]
  particleUpdateScript [&DynamicMaterialParameters(
    Index 0 Param 2=1
    Index 0 Param 3=1
    Index 0 Param 4=1
    Module.Param0WriteEnabled=true
    Module.Param1WriteEnabled=true
    Module.Param2WriteEnabled=true
    Module.Param3WriteEnabled=true
   )]
  rendererClasses [&NiagaraMeshRendererProperties]
  renderers [{rendererClass=&NiagaraMeshRendererProperties, rendererIndex=0}]
 }
 Projectile_Mesh {
  emitterUpdateScript [
   &EmitterState(Life Cycle Mode=::NewEnumerator1, Loop Behavior=::NewEnumerator1, Loop Duration Mode=::NewEnumerator1, UseLoopDelay=false)
   &SpawnBurst_Instantaneous(Use Spawn Probability=false, UseLoopCountLimit=false)
  ]
  particleSpawnScript [
   &InitializeParticle(
    Color=<rgba 10 0 0 1>
    Mesh Scale=<1.25 1 1>
    Mesh Scale Mode=::NewEnumerator4
    Sprite Size Mode=::NewEnumerator3
    Uniform Sprite Size=500
    UsePositionOffset=false
    Write Material Random=true
   )
   &InitializeParticle(
    Color=<rgba 0 35.1565 50 1>
    Mesh Scale Mode=::NewEnumerator4
    Sprite Size Mode=::NewEnumerator3
    Uniform Sprite Size=5
    UsePositionOffset=false
    Write Material Random=true
   ) {enabled=false, module=InitializeParticle002}
   &SystemLocation(MaskSpawnGroup=false, Offset Coordinate Space=::World)
   &InitializeParticle(
    Color Mode=::NewEnumerator0
    Lifetime=&Lerp_Float(
     A=0.1
     B=1
     Alpha=&RemapRange(Input Value To Remap=$User.ScanLength, Input Min=0, Input Max=1000, Output Min=0, Output Max=1)
    )
    Mesh Scale Mode=::NewEnumerator4
    Sprite Size Mode=::NewEnumerator3
    UsePositionOffset=false
    Write Material Random=true
   ) {enabled=false, module=InitializeParticle001}
  ]
  particleUpdateScript [
   &ParticleState(Kill Particles When Lifetime Has Elapsed=false, Let Infinitely Lived Particles Die When Emitter Deactivates=true)
   &DynamicMaterialParameters(
    Index 0 Param 1=$System.Muzzle_Pos_X
    Index 0 Param 2=$System.Muzzle_Pos_Y
    Index 0 Param 3=$System.Muzzle_Pos_Z
    Index 0 Param 4=&FloatFromCurve(FloatCurve=DataInterface() {propertyValues {Curve {keys [<at 0 0>,<at 1 10>]}}}, CurveIndex=$Particles.Age, Scale Curve=1000)
    Module.Param0WriteEnabled=true
    Module.Param1WriteEnabled=true
    Module.Param2WriteEnabled=true
    Module.Param3WriteEnabled=true
   )
   &ScaleMeshSize(Scale Factor=&MakeVector(
     X=&ClampFloat(
      Float=&Divide_Float(
       A=&VectorLength(Vector=&Subtract_Vector(A=Linked() {linkedVariable {name=User.MuzzlePosition, type=NiagaraPosition}}, B=&SimulationPosition()))
       B=200
      )
      Min=0.1
      Max=20
     )
     Y=1
     Z=1
    ))
   &ScaleMeshSize(Scale Factor=&MakeVector(
     X=&ClampFloat(
      Float=&Divide_Float(
       A=&VectorLength(Vector=&Subtract_Vector(
         A=Linked() {linkedVariable {name=User.MuzzlePosition, type=NiagaraPosition}}
         B=Linked() {linkedVariable {name=Engine.Owner.Position, type=NiagaraPosition}}
        ))
       B=100
      )
      Min=0.1
      Max=10
     )
     Y=1
     Z=1
    )) {enabled=false, module=ScaleMeshSize001}
   &UpdateMeshOrientation(
    Facing Direction=&NormalizeVector(
     Vector To Normalize=&Subtract_Vector(
      A=Linked() {linkedVariable {name=Engine.Owner.Position, type=NiagaraPosition}}
      B=Linked() {linkedVariable {name=User.MuzzlePosition, type=NiagaraPosition}}
     )
     Safe Normalize=true
     Safe Fallback Threshold=0
     Safe Fallback Vector=<0 0 0.001>
    )
    Facing Interpolation=false
    Orientation Method=::NewEnumerator1
    Override Quat=false
    TransformFacing=false
    TransformUp=false
   )
   /SetVariables(Particles.Position=&Lerp_Vector(
     A=Linked() {linkedVariable {name=User.MuzzlePosition, type=NiagaraPosition}}
     B=Linked() {linkedVariable {name=Engine.Owner.Position, type=NiagaraPosition}}
     Alpha=&Lerp_Float(
      A=&FloatFromCurve(
       FloatCurve=DataInterface() {propertyValues {Curve {keys [<at 0 0>, {arriveTangent=15.1596, leaveTangent=15.1596, time=0.1, value=1}]}}}
       CurveIndex=$Particles.Age
      )
      B=&FloatFromCurve(
       FloatCurve=DataInterface() {propertyValues {Curve {keys [{arriveTangent=0.7022, time=0, value=0}, {arriveTangent=5.581, leaveTangent=5.581, time=1, value=1}]}}}
       CurveIndex=$Particles.Age
      )
      Alpha=&RemapRange(
       Input Value To Remap=&VectorLength(Vector=&Subtract_Vector(
         A=Linked() {linkedVariable {name=User.MuzzlePosition, type=NiagaraPosition}}
         B=Linked() {linkedVariable {name=Engine.Owner.Position, type=NiagaraPosition}}
        ))
       Input Min=100
       Input Max=15000
       Output Min=0
       Output Max=1
      )
     )
    )) {enabled=false, isSetParameters=true, module=SetVariables_#2}
   /SetVariables(Particles.Position=&Lerp_Vector(
     A=Linked() {linkedVariable {name=User.MuzzlePosition, type=NiagaraPosition}}
     B=Linked() {linkedVariable {name=Engine.Owner.Position, type=NiagaraPosition}}
     Alpha=&FloatFromCurve(
      FloatCurve=DataInterface() {propertyValues {Curve {keys [{arriveTangent=2.0655, time=0, value=0}, {arriveTangent=0.2891, leaveTangent=0.2891, time=0.17, value=1}]}}}
      CurveIndex=$Particles.Age
     )
    )) {isSetParameters=true}
   &ScaleColor(
    Scale Alpha=&FloatFromCurve(FloatCurve=DataInterface() {propertyValues {Curve {keys [<at 0 0>,<at 0.3 0>]}}}, CurveIndex=$Particles.Age)
    ScaleA=true
    ScaleRGB=false
   )
   &ScaleMeshSize(Scale Factor=&VectorFromFloat(Value=&MakeCustomFloatFromBool(
      Bool=$User.IsClient
      True Float=1
      False Float=&Lerp_Float(
       A=0
       B=1
       Alpha=&FloatFromCurve(FloatCurve=DataInterface() {propertyValues {Curve {keys [<at 0.1 0>,<at 0.15 1>]}}})
      )
     ))) {module=ScaleMeshSize002}
   &KillParticles(Kill Particles=&Boolean_OrOperation(
     Bool1=&Boolean_AndOperation(Bool1=$User.Kill Proj, Bool2=&SetBoolByFloatComparison(A=$Particles.Age, B=0.15))
     Bool2=&Boolean_NotOperation(Bool=$User.IsClient)
    ))
   &ScaleColor(
    Scale RGB=&VectorFromFloat(Value=&FNTimeOfDayScalar(Night Scale=1, Morning Scale=0.25, Day Scale=0.5, Evening Scale=0.25))
    ScaleA=true
    ScaleRGB=true
   ) {module=ScaleColor001}
  ]
  rendererClasses [&NiagaraMeshRendererProperties]
  renderers [{rendererClass=&NiagaraMeshRendererProperties, rendererIndex=0}]
 }
 SmokeTrail {
  bEnabled=false
  emitterUpdateScript [
   &EmitterState(Life Cycle Mode=::NewEnumerator1, Loop Duration=1, UseLoopDelay=true)
   &SpawnRate(SpawnRate=100, Use Spawn Probability=false)
  ]
  particleSpawnScript [
   /SetVariables(Particles.Acceleration:Vector3f=<0 0 20>) {enabled=false, isSetParameters=true}
   &InitializeParticle(
    Color=<rgba 5 0 0 1>
    Lifetime=$User.SmokeTrailLifetime
    Ribbon Width=6
    Ribbon Width Mode=::NewEnumerator1
    UsePositionOffset=false
    Write Material Random=true
   )
   &AddVelocity(Velocity=<1000 0 0>) {enabled=false}
  ]
  particleUpdateScript [
   /SetVariables(Particles.ScaledRibbonWidth:NiagaraFloat=&Multiply_Float(
     A=$Particles.RibbonWidth
     B=&FloatFromCurve(FloatCurve=DataInterface() {propertyValues {Curve {keys [<at 0 1 RCIM_Cubic>, {arriveTangent=7.2482, leaveTangent=7.2482, time=0.5, value=3.0109}]}}})
    )) {enabled=false, isSetParameters=true}
   &ParticleState()
   &AccelerationForce(Acceleration=$Particles.Acceleration) {enabled=false}
   &SolveForcesAndVelocity(Clamp Velocity=false, Limit Acceleration=false)
   &ScaleColor(
    Scale Alpha=&FloatFromCurve(
     FloatCurve=DataInterface() {propertyValues {Curve {keys [<at 0 0 RCIM_Cubic>,<at 0.5 1 RCIM_Cubic>]}}}
     CurveIndex=&ReturnNormalizedExecIndex()
     Scale Curve=&OneMinusFloat(Float=$Particles.NormalizedAge)
    )
    Scale RGB=&VectorFromCurve(
     VectorCurve=DataInterface() {propertyValues {XCurve {keys [<at 0.4 6>,<at 0.6 0.5>]}, YCurve {keys [<at 0.4 1.2>,<at 0.6 0.4>]}, ZCurve {keys [<at 0.4 0.4>,<at 0.6 0.15>]}}}
     Scale Curve=&VectorFromFloat(Value=&ReturnNormalizedExecIndex())
    )
    ScaleA=true
    ScaleRGB=false
   )
   &DynamicMaterialParameters(
    Index 0 Param 1=&FloatFromCurve(
     FloatCurve=DataInterface() {propertyValues {Curve {keys [
         <at 0 1 RCIM_Cubic>
         {arriveTangent=-1.0034, interpMode=RCIM_Cubic, leaveTangent=-1.0034, time=0.5114, value=0.7482}
         {interpMode=RCIM_Cubic, leaveTangent=-2.1207, time=0.9966, value=0}
        ]}}}
     Scale Curve=5
    )
    Index 0 Param 2=2
    Index 0 Param 3=10
    Index 0 Param 4=2
    Module.Param0WriteEnabled=true
    Module.Param1WriteEnabled=true
    Module.Param2WriteEnabled=true
    Module.Param3WriteEnabled=true
   )
   &DynamicMaterialParameters(
    Index 0 Param 1=&FloatFromCurve(FloatCurve=DataInterface() {propertyValues {Curve {keys [
         <at 0 1 RCIM_Cubic>
         {arriveTangent=-1.0034, interpMode=RCIM_Cubic, leaveTangent=-1.0034, time=0.5114, value=0.7482}
         {interpMode=RCIM_Cubic, leaveTangent=-2.1207, time=0.9966, value=0}
        ]}}})
    Index 0 Param 2=1
    Index 0 Param 3=1
    Index 0 Param 4=1
    Module.Param0WriteEnabled=true
    Module.Param1WriteEnabled=true
    Module.Param2WriteEnabled=true
    Module.Param3WriteEnabled=true
   ) {enabled=false, module=DynamicMaterialParameters001}
  ]
  rendererClasses [&NiagaraRibbonRendererProperties]
  renderers [{rendererClass=&NiagaraRibbonRendererProperties}]
 }
}
summary {
 systemName=NS_GuineaPig_Prj
 emitters [
  {bEnabled=false, emitterName=Bullet, rendererClasses [&NiagaraMeshRendererProperties]}
  {bEnabled=false, emitterName=Particle_Emitter, rendererClasses [&NiagaraMeshRendererProperties]}
  {bEnabled=false, emitterName=SmokeTrail, rendererClasses [&NiagaraRibbonRendererProperties]}
  {emitterName=Projectile_Mesh, rendererClasses [&NiagaraMeshRendererProperties]}
  {emitterName=Light, rendererClasses [&NiagaraLightRendererProperties]}
  {emitterName=Fake_Proj_Mesh, rendererClasses [&NiagaraMeshRendererProperties]}
  {
   emitterName=ClientSide_Proj
   rendererClasses [&NiagaraSpriteRendererProperties, &NiagaraMeshRendererProperties]
  }
  {emitterName=Light001, rendererClasses [&NiagaraSpriteRendererProperties]}
 ]
 userVariables [
  User.Bool2_Kill Particles {type=NiagaraBool}
  User.Impact Location {type=Vector3f, description="", defaultValue {struct=Vector3f, value=<0 0 0>}}
  User.IsClient {type=NiagaraBool}
  User.Kill Proj {type=NiagaraBool}
  User.MuzzlePosition {type=NiagaraPosition, description="", defaultValue {}}
  User.MuzzlePosition001 {type=Vector3f, description="", defaultValue {struct=Vector3f, value=<0 0 0>}}
  User.RibbonIDOffset {type=NiagaraInt32}
  User.ScanLength {type=NiagaraFloat, defaultValue=0, description=""}
  User.SmokeTrailLifetime {type=NiagaraFloat, defaultValue=0, description=""}
  User.Trail End Position {type=NiagaraPosition, description="", defaultValue {}}
 ]
}
